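# Parent character material for the MarineCarve Ollie backpack skin, adding a masked night glow between the shared character PreFX and PostFX functions.
# Time-of-day functions drive the tinted emissive (EmissiveMask G, optional inverted Fresnel) and darken base color at night (mask R).
# An IsFrontEnd parameter and a time-based loop, shared through named reroutes, provide a front-end preview that pulses on every other cycle with soft edges.
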Material /BRCosmetics/Accessories/FORT_Backpacks/Backpack_F_MED_MarineCarve/Skins/MarineCarve_Ollie/Materials/M_MarineCarve_Ollie_Parent
graph {
 properties {
  bUseMaterialAttributes=true
  bUsedWithMeshDeformer=true
  bUsedWithMorphTargets=true
  bUsedWithSkeletalMesh=true
  pixelDepthOffsetMode=PDOM_Legacy
  shadingModel=MSM_FromMaterialExpression
  usedShadingModels=Default Lit | Subsurface | Clear Coat | Subsurface Profile | Cloth
 }
 nodes [
  n1 = MaterialFunctionCall() {materialFunction=&MF_CharacterParent_2023_PreFX}
  n2 = GetMaterialAttributes(None=@n1.Result) {attributeGetTypes [#1, #2, #3]}
  n3 = Constant()
  n4 = MaterialFunctionCall(
   Color_Morning=@n3
   Color_Day=@n3
   Color_Evening=@n3
   Color_Night=ScalarParameter() {defaultValue=1.5, group=NightGlow, parameterName=NightGlow_Intensity}
  ) {materialFunction=&Color_Time_of_Day}
  n5 = Multiply(A=ScalarParameter() {defaultValue=1.5, parameterName=FE Intensity}, B=NamedRerouteUsage())
  n6 = NamedRerouteUsage()
  n7 = TextureSampleParameter2D() {group=NightGlow, parameterName=EmissiveMask, samplerType=SAMPLERTYPE_Masks, texture=&T_BlackMask}
  n8 = OneMinus(None=Fresnel(ExponentIn=Constant() {r=1}))
  n9 = Constant() {r=1}
  n10 = Multiply(
   A=Multiply(A=LinearInterpolate(A=@n4.Result, B=@n5, Alpha=@n6), B=Power(Base=@n7.G))
   B=Multiply(
    A=VectorParameter() {parameterName=NightGlow_Color, defaultValue=<rgba 1 0.08 0.54 1>}.RGB
    B=StaticSwitchParameter(True=@n8, False=@n9) {parameterName=Use Fresnel}
   )
  )
  n11 = Constant() {r=1}
  n12 = Constant()
  n13 = NamedRerouteUsage()
  n14 = LinearInterpolate(
   B=LinearInterpolate(
    A=MaterialFunctionCall(Value_Morning=@n11, Value_Day=@n11, Value_Evening=@n11, Value_Night=@n12) {materialFunction=&Scalar_Time_of_Day}.Result
    B=LinearInterpolate(Alpha=@n13) {constA=1, constB=0}
    Alpha=NamedRerouteUsage()
   )
   Alpha=@n7.R
  ) {constA=1}
  MaterialFunctionCall(
   Material=SetMaterialAttributes!(
    MaterialAttributes=@n2.MaterialAttributes
    Emissive Color=Add(A=@n2.Emissive Color, B=@n10)
    Base Color=Multiply(A=@n2.Base Color, B=@n14)
    Shading Model
   ) {attributeSetTypes [#1, #2, #3]}
   M=@n1.M
   Pre Skinned Local Position=@n1.M
   Distance Blend=@n1.M
   Fresnel=@n1.M
  ) {materialFunction=&MF_CharacterParent_2023_PostFX}
  n15 = Divide(A=Time(), B=ScalarParameter() {defaultValue=2.5, parameterName="FE Loop`Duration"}) {constB=5}
  n16 = Modulo!(A=@n15, B=Constant() {r=2})
  n17 = Frac(None=@n15)
  n18 = Power(Base=Abs(None=ConstantBiasScale(None=@n17) {bias=-0.5, scale=2})) {constExponent=20}
  NamedRerouteDeclaration(None=Multiply(A=Floor(None=@n16), B=OneMinus(None=@n18))) {name=FE Loop, nodeColor=<rgba 0 0.4141 1 1>}
  NamedRerouteDeclaration(None=ScalarParameter() {parameterName=IsFrontEnd}) {name=IsFrontEnd, nodeColor=<rgba 0.7344 1 0 1>}
 ]
}
